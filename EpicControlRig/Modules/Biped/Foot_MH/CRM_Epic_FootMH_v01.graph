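# Modular Control Rig foot module adding reverse-foot IK, a Ball control and optional FK toe controls to a biped leg.
# Construction builds nested pivot nulls for ball, toe tip, banking and heel.
# Heel and Toe Tip controls plus float channels rotate those pivots, driving the parent's IK effector, with FABRIK and aim posing foot and ball.
# IK/FK switches reset values and request auto-keys; backwards solve copies toe bones to FK controls.
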ControlRigBlueprint /EpicControlRig/Modules/Biped/Foot_MH/CRM_Epic_FootMH_v01
graphs [
 ConstructionGraph {nodes [
   RigUnit_PrepareForExecution = &RigUnit_PrepareForExecution()
   n1 = /Script/ControlRig.RigUnit_ResolveArrayConnector(Connector=(Type=Connector,Name="Toe Joints"), SkipSocket=False) {is=struct RigVMUnitNode}
   RigVMVariableNode(Value=@n1.Result, ExecuteContext=@RigUnit_PrepareForExecution.ExecutePin) {variable=Bones}
   > RigVMVariableNode(Value=@n1.bIsConnected) {variable=Has Toes}
   > RigVMVariableNode(Value=/Script/ControlRig.RigUnit_ResolveConnector(Connector=(Type=Connector,Name="Foot Joint"), SkipSocket=False) {is=struct RigVMUnitNode}.Result) {variable=Foot Joint}
   > RigVMVariableNode(Value=&RigUnit_ResolveConnector(Connector=(Type=Connector,Name="Ball Joint"), SkipSocket=False).Result) {variable=Ball Joint}
   DISPATCH_RigDispatch_GetModuleMetadata = RigVMDispatchNode(Default=(Type=Bone,Name="None"), Name=IK Null, NameSpace=Parent) {template="DISPATCH_RigDispatch_GetModuleMetadata(in Name,in NameSpace,in Default,out Value,out Found)"}
   > RigVMVariableNode(Value=@DISPATCH_RigDispatch_GetModuleMetadata.Value) {variable=Target Effector}
   > RigVMVariableNode(Value=RigVMDispatchNode(Default=(Type=Bone,Name="None"), Name=IK Control, NameSpace=Parent) {template="DISPATCH_RigDispatch_GetModuleMetadata(in Name,in NameSpace,in Default,out Value,out Found)"}.Value) {variable=Target IK Control}
   > DISPATCH_RigVMDispatch_ArrayReset = RigVMDispatchNode(Array=RigVMVariableNode() {variable=BK Buffers}.Value) {template="DISPATCH_RigVMDispatch_ArrayReset(io Array)"}
   > RigVMDispatchNode(Array=RigVMVariableNode() {variable=BK Controls}.Value) {template="DISPATCH_RigVMDispatch_ArrayReset(io Array)"}
   > RigVMDispatchNode(Array=RigVMVariableNode() {variable=FK Controls}.Value) {template="DISPATCH_RigVMDispatch_ArrayReset(io Array)"}
   > RigVMVariableNode(Value=RigVMVariableNode() {variable=Ball Joint}.Value) {variable=Ball Pivot Bone}
   > RigVMFunction_Sequence = RigVMAggregateNode()
   RigVMVariableNode(Value=RigVMVariableNode() {variable=Toe Tip Control}.Value, ExecuteContext=@RigVMFunction_Sequence.G) {variable=Channels Control}
   n2 = RigVMVariableNode() {variable=Channels Control}
   > n3 = /Script/ControlRig.RigUnit_HierarchyAddAnimationChannelFloat(
    InitialValue=0
    LimitsEnabled=(Enabled=(bMinimum=false,bMaximum=false))
    MaximumValue=90
    MinimumValue=-90
    Name=Ball Pivot
    Parent=@n2.Value
   ) {
    is=struct RigVMUnitNode
    template="SpawnAnimationChannel::Execute(in InitialValue,in MinimumValue,in MaximumValue,in LimitsEnabled,in Parent,in Name,out Item)"
   }
   > RigVMVariableNode(Value=@n3.Item) {variable=Ball Pivot Channel}
   > n4 = &RigUnit_HierarchyAddAnimationChannelFloat(
    InitialValue=0
    LimitsEnabled=(Enabled=(bMinimum=false,bMaximum=false))
    MaximumValue=90
    MinimumValue=-90
    Name=Heel Swing
    Parent=@n2.Value
   ) {template="SpawnAnimationChannel::Execute(in InitialValue,in MinimumValue,in MaximumValue,in LimitsEnabled,in Parent,in Name,out Item)"}
   > RigVMVariableNode(Value=@n4.Item) {variable=Heel Swing Channel}
   n5 = RigVMVariableNode() {variable=Ball Joint}
   n6 = &RigUnit_GetTransform(bInitial=False, Item=@n5.Value)
   n7 = &RigUnit_HierarchyAddNull(
    Name=Ball Null
    Space=GlobalSpace
    Transform=@n6.Transform
    Parent=RigVMVariableNode() {variable=Foot Joint}.Value
    ExecutePin=@RigVMFunction_Sequence.E
   )
   > RigVMVariableNode(Value=@n7.Item) {variable=Ball Null}
   EndsWith = /Script/RigVM.RigVMFunction_EndsWith(Ending=_r, Name=@n5.Value.Name) {is=struct RigVMUnitNode, template="EndsWith::Execute(in Name,in Ending,out Result)"}
   > RigVMVariableNode(Value=RigVMDispatchNode(
     Condition=@EndsWith.Result
     True=/Script/RigVM.RigVMFunction_MathQuaternionMul(B=(X=-0.5,Y=0.5,Z=0.5,W=0.5), A=@n6.Transform.Rotation) {is=struct RigVMUnitNode, template="Multiply::Execute(in A,in B,out Result)"}.Result
     False=&RigVMFunction_MathQuaternionMul(B=(X=-0.5,Y=-0.5,Z=0.5,W=-0.5), A=@n6.Transform.Rotation) {template="Multiply::Execute(in A,in B,out Result)"}.Result
    ) {template="DISPATCH_RigVMDispatch_If(in Condition,in True,in False,out Result)"}.Result) {variable=FK Rotation Offset}
   n8 = RigVMVariableNode() {variable=Shape Settings}
   Get Control Color From Metadata = RigVMFunctionReferenceNode(Color Override=@n8.Value.Color)
   > n9 = /Script/ControlRig.RigUnit_HierarchyAddControlTransform(
    InitialValue=(Rotation=(X=0,Y=0,Z=0,W=1),Translation=(X=0,Y=0,Z=0),Scale3D=(X=1,Y=1,Z=1))
    Name=Ball
    OffsetSpace=GlobalSpace
    OffsetTransform=(Rotation=(X=0,Y=0,Z=0,W=1),Translation=(X=0,Y=0,Z=0),Scale3D=(X=1,Y=1,Z=1))
    Settings="(InitialSpace=LocalSpace,bUsePreferredRotationOrder=False,PreferredRotationOrder=YZX,Limits=(LimitTranslationX=(bMinimum=False,bMaximum=False),LimitTranslationY=(bMinimum=False,bMaximum=False),LimitTranslationZ=(bMinimum=False,bMaximum=False),LimitPitch=(b ...(truncated)"
    Settings.Shape.Color=@"Get Control Color From Metadata".Result
    OffsetTransform.Scale3D=@n6.Transform.Scale3D
    OffsetTransform.Translation=@n6.Transform.Translation
    Parent=@n7.Item
    Settings.Shape.Name=@n8.Value.Name
    Settings.Shape.Transform=@n8.Value.Transform
    Settings.Shape.bVisible=@n8.Value.bVisible
   ) {
    is=struct RigVMUnitNode
    template="SpawnControl::Execute(in InitialValue,in Settings,in OffsetTransform,in OffsetSpace,in Parent,in Name,out Item)"
   }
   > RigVMVariableNode(Value=@n9.Item) {variable=Ball Control}
   n10 = RigVMVariableNode() {variable=Pivot Shape Settings}
   Get Control Color From Metadata_1 = RigVMFunctionReferenceNode(Color Override=@n10.Value.Color)
   n11 = &RigUnit_HierarchyAddNull(
    Name=IK Foot Space
    Space=LocalSpace
    Transform=(Rotation=(X=0,Y=0,Z=0,W=1),Translation=(X=0,Y=0,Z=0),Scale3D=(X=1,Y=1,Z=1))
    Parent=RigVMVariableNode() {variable=Target IK Control}.Value
    ExecutePin=@RigVMFunction_Sequence.A
   )
   SpawnControl = &RigUnit_HierarchyAddControlTransform(
    InitialValue=(Rotation=(X=0,Y=0,Z=0,W=1),Translation=(X=0,Y=0,Z=0),Scale3D=(X=1,Y=1,Z=1))
    Name=Toe Tip
    OffsetSpace=GlobalSpace
    Settings="(InitialSpace=LocalSpace,bUsePreferredRotationOrder=true,PreferredRotationOrder=ZYX,Limits=(LimitTranslationX=(bMinimum=False,bMaximum=False),LimitTranslationY=(bMinimum=False,bMaximum=False),LimitTranslationZ=(bMinimum=False,bMaximum=False),LimitPitch=(bM ...(truncated)"
    Settings.Shape.Color=@"Get Control Color From Metadata_1".Result
    OffsetTransform=&RigUnit_GetTransform(bInitial=False, Item=RigVMVariableNode() {variable=Toe Tip Pivot Null}.Value).Transform
    Parent=@n11.Item
    Settings.Shape.Name=@n10.Value.Name
    Settings.Shape.Transform=@n10.Value.Transform
    Settings.Shape.bVisible=@n10.Value.bVisible
    ExecutePin=@RigVMFunction_Sequence.B
   ) {template="SpawnControl::Execute(in InitialValue,in Settings,in OffsetTransform,in OffsetSpace,in Parent,in Name,out Item)"}
   > RigVMVariableNode(Value=@SpawnControl.Item) {variable=Toe Tip Control}
   > RigVMDispatchNode(Array=RigVMVariableNode() {variable=IK Controls}.Value, Element=@SpawnControl.Item) {template="DISPATCH_RigVMDispatch_ArrayAdd(io Array,in Element,out Index)"}
   n12 = RigVMVariableNode() {variable=Ball Joint}
   > Set_Metadata = RigVMDispatchNode(
    Name=Delta
    NameSpace=Self
    Value=&RigVMFunction_MathTransformMakeRelative(
     Global=&RigUnit_GetTransform(bInitial=true, Item=@SpawnControl.Item).Transform
     Parent=&RigUnit_GetTransform(bInitial=true, Item=@n12.Value).Transform
    ).Local
    Item=@SpawnControl.Item
   ) {template="DISPATCH_RigDispatch_SetMetadata(in Item,in Name,in NameSpace,in Value,out Success)"}
   RigUnit_GetTransform = &RigUnit_GetTransform(bInitial=False, Item=RigVMVariableNode() {variable=Ball Pivot Null}.Value)
   n13 = &RigUnit_HierarchyAddNull(
    Name=Ball IK
    Space=GlobalSpace
    Transform=@RigUnit_GetTransform.Transform
    Parent=RigVMVariableNode() {variable=Heel Lift Null}.Value
    ExecutePin=@RigVMFunction_Sequence.D
   )
   > RigVMVariableNode(Value=@n13.Item) {variable=Ball IK Null}
   > n14 = &RigUnit_HierarchyAddNull(
    Name=Toe IK
    Space=GlobalSpace
    Transform=&RigUnit_GetTransform(bInitial=False, Item=RigVMVariableNode() {variable=Toe Tip Pivot Null}.Value).Transform
    Parent=RigVMVariableNode() {variable=Heel Roll Null}.Value
   )
   > RigVMVariableNode(Value=@n14.Item) {variable=Toe IK Null}
   > n15 = &RigUnit_HierarchyAddNull(
    Name=IK
    Space=GlobalSpace
    Transform=&RigUnit_GetTransform(bInitial=False, Item=RigVMVariableNode() {variable=Target Effector}.Value).Transform
    Parent=RigVMVariableNode() {variable=Heel Lift Null}.Value
   )
   > RigVMVariableNode(Value=@n15.Item) {variable=IK Null}
   n16 = &RigVMFunction_ControlFlowBranch(Condition=RigVMVariableNode() {variable=Has Toes}.Value, ExecuteContext=@RigVMFunction_Sequence.F)
   n17 = RigVMDispatchNode(ExecuteContext=@n16.True, Array=RigVMVariableNode() {variable=Bones}.Value) {template="DISPATCH_RigVMDispatch_ArrayIterator(in Array,out Element,out Index,out Count,out Ratio)"}
   Int to Name_2 = /Script/RigVM.RigVMFunction_MathIntToName(PaddedSize=0, Number=@n17.Index) {is=struct RigVMUnitNode, template="Int to Name::Execute(in Number,in PaddedSize,out Result)"}
   n18 = RigVMVariableNode() {variable=Shape Settings}
   Get Control Color From Metadata_3 = RigVMFunctionReferenceNode(Color Override=@n18.Value.Color)
   > n19 = &RigUnit_HierarchyAddControlTransform(
    InitialValue=(Rotation=(X=0,Y=0,Z=0,W=1),Translation=(X=0,Y=0,Z=0),Scale3D=(X=1,Y=1,Z=1))
    OffsetSpace=GlobalSpace
    Parent=(Type=Bone,Name="None")
    Settings="(InitialSpace=LocalSpace,bUsePreferredRotationOrder=False,PreferredRotationOrder=YZX,Limits=(LimitTranslationX=(bMinimum=False,bMaximum=False),LimitTranslationY=(bMinimum=False,bMaximum=False),LimitTranslationZ=(bMinimum=False,bMaximum=False),LimitPitch=(b ...(truncated)"
    Name=/Script/RigVM.RigVMFunction_NameConcat(A="FK ", B=@"Int to Name_2".Result) {is=struct RigVMUnitNode, template="Concat::Execute(in A,in B,out Result)"}.Result
    Settings.Shape.Color=@"Get Control Color From Metadata_3".Result
    OffsetTransform=&RigUnit_GetTransform(bInitial=False, Item=@n17.Element).Transform
    Settings.Shape.Name=@n18.Value.Name
    Settings.Shape.Transform=@n18.Value.Transform
    Settings.Shape.bVisible=@n18.Value.bVisible
   ) {template="SpawnControl::Execute(in InitialValue,in Settings,in OffsetTransform,in OffsetSpace,in Parent,in Name,out Item)"}
   > RigVMDispatchNode(Element=@n19.Item, Array=RigVMVariableNode() {variable=FK Controls}.Value) {template="DISPATCH_RigVMDispatch_ArrayAdd(io Array,in Element,out Index)"}
   Make_Array = RigVMDispatchNode(
    Values=((Type=None,Name="None"),((),Type=None,Name="None"))
    Values.0=RigVMVariableNode() {variable=Foot Joint}.Value
    Values.1=RigVMVariableNode() {variable=Ball Joint}.Value
   ) {template="DISPATCH_RigVMDispatch_ArrayMake(in Values,out Array)"}
   Get Chain Length = RigVMFunctionReferenceNode(Initial=true, Items=@Make_Array.Array, ExecuteContext=@RigVMFunction_Sequence.H)
   Get_Module_Metadata_1 = RigVMDispatchNode(Default=1, Name=Global Control Scale, NameSpace=Root) {template="DISPATCH_RigDispatch_GetModuleMetadata(in Name,in NameSpace,in Default,out Value,out Found)"}
   Num = RigVMDispatchNode(Array=@Make_Array.Array) {template="DISPATCH_RigVMDispatch_ArrayGetNum(in Array,out Num)"}
   > Set Control Scale = RigVMFunctionReferenceNode(
    Global Scale=@Get_Module_Metadata_1.Value
    Auto Scale Factor=RigVMDispatchNode(
     False=1
     True=&RigVMFunction_MathDoubleDiv(B=17, A=@"Get Chain Length".Length).Result
     Condition=/Script/RigVM.RigVMFunction_MathIntGreater(B=1, A=@Num.Num) {is=struct RigVMUnitNode, template="Greater::Execute(in A,in B,out Result)"}.Result
    ) {template="DISPATCH_RigVMDispatch_If(in Condition,in True,in False,out Result)"}.Result
    Controls=RigVMVariableNode() {variable=FK Controls}.Value
    Scale=RigVMVariableNode() {variable=Control Scale}.Value
   )
   n20 = &RigUnit_GetTransform(bInitial=False, Item=RigVMVariableNode() {variable=Ball Pivot Bone}.Value)
   n21 = /Script/RigVM.RigVMFunction_MathTransformMakeAbsolute(
    Local=(Rotation=(X=0,Y=0,Z=0,W=1),Translation=(X=0,Y=0,Z=0),Scale3D=(X=1,Y=1,Z=1))
    Parent=(Rotation=(X=0,Y=0,Z=0,W=1),Translation=(X=0,Y=0,Z=0),Scale3D=(X=1,Y=1,Z=1))
    Parent.Translation=@n20.Transform.Translation
    Local.Translation=RigVMVariableNode() {variable=Inner Bank Offset}.Value
   ) {is=struct RigVMUnitNode, template="Make Absolute::Execute(in Local,in Parent,out Global)"}
   n22 = &RigUnit_HierarchyAddNull(
    Name=Inner
    Space=GlobalSpace
    Transform=@n21.Global
    Parent=RigVMVariableNode() {variable=Toe Tip Pivot Null}.Value
    ExecutePin=@RigVMFunction_Sequence.C
   )
   > RigVMVariableNode(Value=@n22.Item) {variable=Inner Null}
   n23 = &RigVMFunction_MathTransformMakeAbsolute(
    Local=(Rotation=(X=0,Y=0,Z=0,W=1),Translation=(X=0,Y=0,Z=0),Scale3D=(X=1,Y=1,Z=1))
    Parent=(Rotation=(X=0,Y=0,Z=0,W=1),Translation=(X=0,Y=0,Z=0),Scale3D=(X=1,Y=1,Z=1))
    Parent.Translation=@n20.Transform.Translation
    Local.Translation=RigVMVariableNode() {variable=Outer Bank Offset}.Value
   ) {template="Make Absolute::Execute(in Local,in Parent,out Global)"}
   > n24 = &RigUnit_HierarchyAddNull(Name=Outer, Space=GlobalSpace, Parent=@n22.Item, Transform=@n23.Global)
   > RigVMVariableNode(Value=@n24.Item) {variable=Outer Null}
   n25 = RigVMVariableNode() {variable=Pivot Shape Settings}
   Get Control Color From Metadata_2 = RigVMFunctionReferenceNode(Color Override=@n25.Value.Color)
   n26 = &RigVMFunction_MathTransformMakeAbsolute(
    Local=(Rotation=(X=0,Y=0,Z=0,W=1),Translation=(X=0,Y=0,Z=0),Scale3D=(X=1,Y=1,Z=1))
    Parent=(Rotation=(X=0,Y=0,Z=0,W=1),Translation=(X=0,Y=0,Z=0),Scale3D=(X=1,Y=1,Z=1))
    Parent.Translation=&RigUnit_GetTransform(bInitial=False, Item=RigVMVariableNode() {variable=Ball Pivot Bone}.Value).Transform.Translation
    Local.Translation=RigVMVariableNode() {variable=Heel Pivot Offset}.Value
   ) {template="Make Absolute::Execute(in Local,in Parent,out Global)"}
   > n27 = &RigUnit_HierarchyAddControlTransform(
    InitialValue=(Rotation=(X=0,Y=0,Z=0,W=1),Translation=(X=0,Y=0,Z=0),Scale3D=(X=1,Y=1,Z=1))
    Name=Heel
    OffsetSpace=GlobalSpace
    Settings="(InitialSpace=LocalSpace,bUsePreferredRotationOrder=true,PreferredRotationOrder=ZYX,Limits=(LimitTranslationX=(bMinimum=False,bMaximum=False),LimitTranslationY=(bMinimum=False,bMaximum=False),LimitTranslationZ=(bMinimum=False,bMaximum=False),LimitPitch=(bM ...(truncated)"
    Settings.Shape.Color=@"Get Control Color From Metadata_2".Result
    OffsetTransform=@n26.Global
    Settings.Shape.Name=@n25.Value.Name
    Settings.Shape.Transform=@n25.Value.Transform
    Settings.Shape.bVisible=@n25.Value.bVisible
    Parent=RigVMVariableNode() {variable=Outer Null}.Value
   ) {template="SpawnControl::Execute(in InitialValue,in Settings,in OffsetTransform,in OffsetSpace,in Parent,in Name,out Item)"}
   > RigVMVariableNode(Value=@n27.Item) {variable=Heel Control}
   > RigVMDispatchNode(Array=RigVMVariableNode() {variable=IK Controls}.Value, Element=@n27.Item) {template="DISPATCH_RigVMDispatch_ArrayAdd(io Array,in Element,out Index)"}
   n28 = RigVMVariableNode() {variable=Foot Joint}
   > Set_Metadata_1 = RigVMDispatchNode(
    Name=Delta
    NameSpace=Self
    Value=&RigVMFunction_MathTransformMakeRelative(
     Global=&RigUnit_GetTransform(bInitial=true, Item=@n27.Item).Transform
     Parent=&RigUnit_GetTransform(bInitial=true, Item=@n28.Value).Transform
    ).Local
    Item=@n27.Item
   ) {template="DISPATCH_RigDispatch_SetMetadata(in Item,in Name,in NameSpace,in Value,out Success)"}
   > n29 = &RigUnit_HierarchyAddNull(Name=Heel Roll, Space=GlobalSpace, Parent=RigVMVariableNode() {variable=Outer Null}.Value, Transform=@n26.Global)
   > RigVMVariableNode(Value=@n29.Item) {variable=Running Parent}
   > RigVMVariableNode(Value=@n29.Item) {variable=Heel Roll Null}
   > n30 = &RigUnit_HierarchyAddNull(
    Name=Heel Lift
    Space=GlobalSpace
    Transform=&RigUnit_GetTransform(bInitial=False, Item=RigVMVariableNode() {variable=Ball Pivot Bone}.Value).Transform
    Parent=@n29.Item
   )
   > RigVMVariableNode(Value=@n30.Item) {variable=Heel Lift Null}
   RigVMRerouteNode(Value=@n20.Transform.Translation)
   RigVMVariableNode() {variable=Target IK Control}
   n31 = &RigUnit_GetTransform(bInitial=False, Item=RigVMVariableNode() {variable=Ball Pivot Bone}.Value)
   n32 = &RigUnit_HierarchyAddNull(Name=Ball Pivot, Space=GlobalSpace, Parent=@n11.Item, ExecutePin=@n11.ExecutePin, Transform=@n31.Transform)
   n33 = &RigVMFunction_MathTransformMakeAbsolute(
    Local=(Rotation=(X=0,Y=0,Z=0,W=1),Translation=(X=0,Y=0,Z=0),Scale3D=(X=1,Y=1,Z=1))
    Parent=(Rotation=(X=0,Y=0,Z=0,W=1),Translation=(X=0,Y=0,Z=0),Scale3D=(X=1,Y=1,Z=1))
    Parent.Translation=@n31.Transform.Translation
    Local.Translation=RigVMVariableNode() {variable=Toe Tip Pivot Offset}.Value
   ) {template="Make Absolute::Execute(in Local,in Parent,out Global)"}
   n34 = &RigUnit_HierarchyAddNull(Name=Toe Tip Pivot, Space=GlobalSpace, Parent=@n32.Item, ExecutePin=@n35.ExecuteContext, Transform=@n33.Global)
   > RigVMVariableNode(Value=@n34.Item) {variable=Toe Tip Pivot Null}
   n35 = RigVMVariableNode(ExecuteContext=@n32.ExecutePin, Value=@n32.Item) {variable=Ball Pivot Null}
  ]}
 New Function_ContainedGraph {nodes [
   n1 = RigVMFunctionEntryNode()
   RigUnit_GetTransform = &RigUnit_GetTransform(Space=LocalSpace, bInitial=true, Item=@n1.Item)
   > RigUnit_SetTransform = &RigUnit_SetTransform(
    Space=LocalSpace
    Weight=1
    bInitial=False
    Item=@n1.Item
    Value=&RigVMFunction_MathTransformMakeAbsolute(Local=@n1.Local, Parent=@RigUnit_GetTransform.Transform) {template="Make Absolute::Execute(in Local,in Parent,out Global)"}.Global
   )
   > RigVMFunctionReturnNode()
  ]}
 New Function_ContainedGraph {nodes [
   RigVMFunctionEntryNode()
   > RigVMFunction_Sequence = RigVMAggregateNode()
   RigVMFunctionReturnNode(ExecuteContext=@RigVMFunction_Sequence.D)
   Set Null Local Transform_0 = RigVMFunctionReferenceNode(
    Local=(Rotation=(X=0,Y=0,Z=0,W=1),Translation=(X=0,Y=0,Z=0),Scale3D=(X=1,Y=1,Z=1))
    Local.Rotation=&RigVMFunction_MathQuaternionFromEuler(Euler=(X=0,Y=0,Z=0), Euler.X=RigVMVariableNode() {variable=Toe Tip Roll}.Value, Euler.Z=RigVMVariableNode() {variable=Toe Tip Swing}.Value).Result
    Item=RigVMVariableNode() {variable=Toe Tip Pivot Null}.Value
    ExecuteContext=@RigVMFunction_Sequence.A
   )
   n1 = RigVMVariableNode() {variable=Bank}
   Clamp = &RigVMFunction_MathDoubleClamp(Maximum=0, Minimum=-90, Value=@n1.Value)
   RigVMFunction_MathQuaternionFromEuler = &RigVMFunction_MathQuaternionFromEuler(
    Euler=(X=0,Y=0,Z=0)
    Euler.Y=RigVMDispatchNode(
     False=@Clamp.Result
     True=&RigVMFunction_MathDoubleClamp(Maximum=90, Value=@n1.Value).Result
     Condition=RigVMVariableNode() {variable=Negative Side}.Value
    ) {template="DISPATCH_RigVMDispatch_If(in Condition,in True,in False,out Result)"}.Result
   )
   > Set Transform_1 = &RigUnit_SetRotation(
    Weight=1
    bInitial=False
    bPropagateToChildren=true
    Value=@RigVMFunction_MathQuaternionFromEuler.Result
    Item=RigVMVariableNode() {variable=Inner Null}.Value
   )
   n2 = RigVMVariableNode() {variable=Bank}
   n3 = &RigVMFunction_MathDoubleClamp(Maximum=90, Value=@n2.Value)
   n4 = &RigVMFunction_MathDoubleClamp(Maximum=0, Minimum=-90, Value=@n2.Value)
   n5 = RigVMVariableNode() {variable=Negative Side}
   > Set Transform_1_1 = &RigUnit_SetRotation(
    Weight=1
    bInitial=False
    Value=&RigVMFunction_MathQuaternionFromEuler(
     Euler=(X=0,Y=0,Z=0)
     Euler.Y=RigVMDispatchNode(False=@n3.Result, True=@n4.Result, Condition=@n5.Value) {template="DISPATCH_RigVMDispatch_If(in Condition,in True,in False,out Result)"}.Result
    ).Result
    Item=RigVMVariableNode() {variable=Outer Null}.Value
   )
   n6 = &RigVMFunction_MathDoubleClamp(Maximum=0, Minimum=-360, Value=RigVMVariableNode() {variable=Ball Roll}.Value)
   Set Null Local Transform_2 = RigVMFunctionReferenceNode(
    Local=(Rotation=(X=0,Y=0,Z=0,W=1),Translation=(X=0,Y=0,Z=0),Scale3D=(X=1,Y=1,Z=1))
    Local.Rotation=&RigVMFunction_MathQuaternionFromEuler(
     Euler=(X=0,Y=0,Z=0)
     Euler.Z=/Script/RigVM.RigVMFunction_MathDoubleNegate(Value=@n6.Result) {is=struct RigVMUnitNode, template="Negate::Execute(in Value,out Result)"}.Result
     Euler.X=RigVMVariableNode() {variable=Heel Twist}.Value
     Euler.Y=RigVMVariableNode() {variable=Heel Swing}.Value
    ).Result
    Item=RigVMVariableNode() {variable=Heel Lift Null}.Value
    ExecuteContext=@RigVMFunction_Sequence.B
   )
   n7 = RigVMVariableNode() {variable=Ball Roll}
   > Set Null Local Transform_3 = RigVMFunctionReferenceNode(
    Local=(Rotation=(X=0,Y=0,Z=0,W=1),Translation=(X=0,Y=0,Z=0),Scale3D=(X=1,Y=1,Z=1))
    Local.Rotation=&RigVMFunction_MathQuaternionFromEuler(
     Euler=(X=0,Y=0,Z=0)
     Euler.X=&RigVMFunction_MathDoubleClamp(Maximum=360, Value=@n7.Value).Result
     Euler.Z=RigVMVariableNode() {variable=Heel Pivot}.Value
    ).Result
    Item=RigVMVariableNode() {variable=Heel Roll Null}.Value
   )
   > Set Null Local Transform_4 = RigVMFunctionReferenceNode(
    Local=(Rotation=(X=0,Y=0,Z=0,W=1),Translation=(X=0,Y=0,Z=0),Scale3D=(X=1,Y=1,Z=1))
    Local.Rotation=&RigVMFunction_MathQuaternionFromEuler(Euler=(X=0,Y=0,Z=0), Euler.Y=RigVMVariableNode() {variable=Ball Pivot}.Value).Result
    Item=RigVMVariableNode() {variable=Ball Pivot Null}.Value
   )
   &RigUnit_SetTransform(
    Weight=1
    bInitial=False
    Value=&RigUnit_GetTransform(Item=(Type=Null,Name="IK"), bInitial=False).Transform
    Item=RigVMVariableNode() {variable=Target Effector}.Value
    ExecutePin=@RigVMFunction_Sequence.C
   )
  ]}
 New Function_ContainedGraph {nodes [
   n1 = RigVMFunctionEntryNode()
   RigVMFunction_Sequence = &RigVMFunction_Sequence(ExecuteContext=@n1.ExecuteContext)
   RigVMVariableNode(Value=0, ExecuteContext=@RigVMFunction_Sequence.A) {variable=Ball Roll}
   > RigVMVariableNode(Value=0) {variable=Heel Twist}
   > RigVMVariableNode(Value=0) {variable=Heel Swing}
   > RigVMVariableNode(Value=0) {variable=Toe Tip Roll}
   > RigVMVariableNode(Value=0) {variable=Bank}
   > RigVMVariableNode(Value=0) {variable=Toe Tip Swing}
   > RigVMVariableNode() {variable=Heel Pivot}
   > RigVMVariableNode() {variable=Ball Pivot}
   RigVMFunctionReturnNode(ExecuteContext=@RigVMFunction_Sequence.B)
  ]}
 New Function_ContainedGraph {nodes [
   n1 = RigVMFunctionEntryNode()
   RigVMFunction_Sequence = &RigVMFunction_Sequence(ExecuteContext=@n1.ExecuteContext)
   n2 = &RigVMFunction_MathQuaternionToEuler(
    RotationOrder=ZYX
    Value=&RigUnit_GetTransform(Space=LocalSpace, bInitial=False, Item=RigVMVariableNode() {variable=Heel Control}.Value).Transform.Rotation
   )
   RigVMVariableNode(ExecuteContext=@RigVMFunction_Sequence.A, Value=@n2.Result.X) {variable=Ball Roll}
   > RigVMVariableNode(Value=RigVMDispatchNode(
     True=&RigVMFunction_MathDoubleNegate(Value=@n2.Result.Y) {template="Negate::Execute(in Value,out Result)"}.Result
     Condition=RigVMVariableNode() {variable=Negative Side}.Value
     False=@n2.Result.Y
    ) {template="DISPATCH_RigVMDispatch_If(in Condition,in True,in False,out Result)"}.Result) {variable=Heel Twist}
   > RigVMVariableNode(Value=@n2.Result.Z) {variable=Heel Pivot}
   n3 = &RigVMFunction_MathQuaternionToEuler(
    RotationOrder=ZYX
    Value=&RigUnit_GetTransform(Space=LocalSpace, bInitial=False, Item=RigVMVariableNode() {variable=Toe Tip Control}.Value).Transform.Rotation
   )
   > RigVMVariableNode(Value=@n3.Result.X) {variable=Toe Tip Roll}
   > RigVMVariableNode(Value=@n3.Result.Y) {variable=Bank}
   > RigVMVariableNode(Value=@n3.Result.Z) {variable=Toe Tip Swing}
   From_String_4 = RigVMDispatchNode(String=Heel Swing) {template="DISPATCH_RigDispatch_FromString(in String,out Result)"}
   n4 = RigVMVariableNode() {variable=Channels Control}
   > RigVMVariableNode(Value=/Script/ControlRig.RigUnit_GetFloatAnimationChannel(bInitial=False, Channel=@From_String_4.Result, Control=@n4.Value.Name) {is=struct RigVMUnitNode, template="GetAnimationChannel::Execute(out Value,in Control,in Channel,in bInitial)"}.Value) {variable=Heel Swing}
   From_String_5 = RigVMDispatchNode(String=Ball Pivot) {template="DISPATCH_RigDispatch_FromString(in String,out Result)"}
   > RigVMVariableNode(Value=&RigUnit_GetFloatAnimationChannel(bInitial=False, Channel=@From_String_5.Result, Control=@n4.Value.Name) {template="GetAnimationChannel::Execute(out Value,in Control,in Channel,in bInitial)"}.Value) {variable=Ball Pivot}
   RigVMFunctionReturnNode(ExecuteContext=@RigVMFunction_Sequence.B)
  ]}
 RigVMFunctionLibrary {nodes [
   Reset Foot Rig Values = RigVMCollapseNode(ExecuteContext=())
   Set Foot Pivots = RigVMCollapseNode(ExecuteContext=())
   Set Foot Values = RigVMCollapseNode(ExecuteContext=())
   Set Null Local Transform = RigVMCollapseNode(ExecuteContext=(), Item=(Type=Null,Name="BK 0"), Local=(Rotation=(X=0,Y=0,Z=0,W=1),Translation=(X=0,Y=0,Z=0),Scale3D=(X=1,Y=1,Z=1)))
  ]}
 RigVMFunction_Sequence_1_ContainedGraph {nodes [
   n1 = RigVMFunctionEntryNode()
   RigVMFunction_Sequence = &RigVMFunction_Sequence(ExecuteContext=@n1.ExecuteContext)
   n2 = &RigVMFunction_Sequence(ExecuteContext=@RigVMFunction_Sequence.B)
   n3 = &RigVMFunction_Sequence(ExecuteContext=@n2.B)
   n4 = &RigVMFunction_Sequence(ExecuteContext=@n3.B)
   n5 = &RigVMFunction_Sequence(ExecuteContext=@n4.B)
   n6 = &RigVMFunction_Sequence(ExecuteContext=@n5.B)
   n7 = &RigVMFunction_Sequence(ExecuteContext=@n6.B)
   RigVMFunctionReturnNode(A=@RigVMFunction_Sequence.A, B=@n2.A, C=@n3.A, D=@n4.A, E=@n5.A, F=@n6.A, G=@n7.A, H=@n7.B)
  ]}
 RigVMFunction_Sequence_1_ContainedGraph {nodes [
   n1 = RigVMFunctionEntryNode()
   RigVMFunction_Sequence = &RigVMFunction_Sequence(ExecuteContext=@n1.ExecuteContext)
   n2 = &RigVMFunction_Sequence(ExecuteContext=@RigVMFunction_Sequence.B)
   n3 = &RigVMFunction_Sequence(ExecuteContext=@n2.B)
   RigVMFunctionReturnNode(A=@RigVMFunction_Sequence.A, B=@n2.A, C=@n3.A, D=@n3.B)
  ]}
 RigVMFunction_Sequence_3_ContainedGraph {nodes [
   n1 = RigVMFunctionEntryNode()
   n2 = &RigVMFunction_Sequence(ExecuteContext=@n1.ExecuteContext)
   n3 = &RigVMFunction_Sequence(ExecuteContext=@n2.B)
   RigVMFunctionReturnNode(B=@n2.A, D=@n3.A, E=@n3.B)
  ]}
 RigVMModel Backwards Solve Graph {nodes [
   n1 = /Script/ControlRig.RigUnit_InverseExecution() {is=struct RigVMUnitNode}
   RigVMFunction_Sequence = &RigVMFunction_Sequence(ExecuteContext=@n1.ExecutePin)
   Negate = &RigVMFunction_MathDoubleNegate(Value=RigVMVariableNode() {variable=Ball Roll}.Value) {template="Negate::Execute(in Value,out Result)"}
   Set Null Local Transform_1_2 = RigVMFunctionReferenceNode(
    Local=(Rotation=(X=0,Y=0,Z=0,W=1),Translation=(X=0,Y=0,Z=0),Scale3D=(X=1,Y=1,Z=1))
    Local.Rotation=/Script/RigVM.RigVMFunction_MathQuaternionFromRotatorV2(Value=RigVMDispatchNode(Elements=(Pitch=0,Yaw=0,Roll=0), Elements.Yaw=@Negate.Result) {template="DISPATCH_RigVMDispatch_MakeStruct(in Elements,out Struct)"}.Struct) {is=struct RigVMUnitNode, template="Cast::Execute(in Value,out Result)"}.Result
    ExecuteContext=@RigVMFunction_Sequence.A
    Item=RigVMVariableNode() {variable=Ball Null}.Value
   )
   /Script/ControlRig.RigUnit_SetTransformItemArray(
    Space=GlobalSpace
    Weight=1
    bInitial=false
    bPropagateToChildren=true
    Transforms=/Script/ControlRig.RigUnit_GetTransformItemArray(Space=GlobalSpace, bInitial=false, Items=RigVMVariableNode() {variable=Bones}.Value) {is=struct RigVMUnitNode}.Transforms
    Items=RigVMVariableNode() {variable=FK Controls}.Value
    ExecutePin=@RigVMFunction_Sequence.B
   ) {is=struct RigVMUnitNode}
  ]}
 RigVMModel Forwards Solve Graph {nodes [
   n1 = /Script/ControlRig.RigUnit_PostBeginExecution() {is=struct RigVMUnitNode}
   Not_Equals = RigVMDispatchNode(A=RigVMVariableNode() {variable=IK Solve}.Value, B=RigVMVariableNode() {variable=Previous IK Solve}.Value) {template="DISPATCH_RigVMDispatch_CoreNotEquals(in A,in B,out Result)"}
   n2 = &RigVMFunction_ControlFlowBranch(Condition=@Not_Equals.Result, ExecuteContext=@n1.ExecutePin)
   n3 = RigVMDispatchNode(Enabled=True, ScreenColor=(R=1,G=1,B=1,A=1), ScreenDuration=0.05, Value=DIFFEREN, ExecuteContext=@n2.True) {template="DISPATCH_RigVMDispatch_Print(in Prefix,in Value,in Enabled,in ScreenDuration,in ScreenColor)"}
   RigVMFunction_Sequence = &RigVMFunction_Sequence(ExecuteContext=@n3.ExecuteContext)
   n4 = &RigVMFunction_ControlFlowBranch(ExecuteContext=@RigVMFunction_Sequence.A, Condition=RigVMVariableNode() {variable=IK Solve}.Value)
   n5 = &RigVMFunction_Sequence(ExecuteContext=@n4.False)
   From_String_3 = RigVMDispatchNode(String=Ball Pivot) {template="DISPATCH_RigDispatch_FromString(in String,out Result)"}
   n6 = RigVMVariableNode() {variable=Channels Control}
   /Script/ControlRig.RigUnit_SetFloatAnimationChannel(Value=0, bInitial=False, Channel=@From_String_3.Result, ExecutePin=@n5.A, Control=@n6.Value.Name) {is=struct RigVMUnitNode, template="SetAnimationChannel::Execute(in Value,in Control,in Channel,in bInitial)"}
   From_String_2 = RigVMDispatchNode(String=Heel Swing) {template="DISPATCH_RigDispatch_FromString(in String,out Result)"}
   > SetAnimationChannel = &RigUnit_SetFloatAnimationChannel(Value=0, bInitial=False, Channel=@From_String_2.Result, Control=@n6.Value.Name) {template="SetAnimationChannel::Execute(in Value,in Control,in Channel,in bInitial)"}
   > &RigUnit_SetTransform(
    Space=LocalSpace
    Value=(Rotation=(X=0,Y=0,Z=0,W=1),Translation=(X=0,Y=0,Z=0),Scale3D=(X=1,Y=1,Z=1))
    Weight=1
    bInitial=False
    bPropagateToChildren=false
    Item=RigVMVariableNode() {variable=Toe Tip Control}.Value
   )
   > &RigUnit_SetTransform(
    Space=LocalSpace
    Value=(Rotation=(X=0,Y=0,Z=0,W=1),Translation=(X=0,Y=0,Z=0),Scale3D=(X=1,Y=1,Z=1))
    Weight=1
    bInitial=False
    bPropagateToChildren=false
    Item=RigVMVariableNode() {variable=Heel Control}.Value
   )
   > Reset Foot Rig Values = RigVMFunctionReferenceNode()
   > Set Foot Pivots_1 = RigVMFunctionReferenceNode()
   n7 = RigVMVariableNode() {variable=Ball Null}
   Set Transform_1 = &RigUnit_SetRotation(
    Weight=1
    bInitial=false
    bPropagateToChildren=true
    Value=&RigUnit_GetTransform(Space=LocalSpace, bInitial=true, Item=@n7.Value).Transform.Rotation
    Item=@n7.Value
    ExecutePin=@n5.B
   )
   > RigUnit_SetRotation = &RigUnit_SetTransform(
    Weight=1
    bInitial=False
    bPropagateToChildren=false
    Value=RigVMVariableNode() {variable=Current Ball Control Transform}.Value
    Item=RigVMVariableNode() {variable=Ball Control}.Value
   )
   n8 = RigVMDispatchNode(
    Array=RigVMDispatchNode(Array=RigVMVariableNode() {variable=FK Controls}.Value) {template="DISPATCH_RigVMDispatch_ArrayClone(in Array,out Clone)"}.Clone
    Other=RigVMVariableNode() {variable=IK Controls}.Value
    ExecuteContext=@RigVMFunction_Sequence.B
   ) {template="DISPATCH_RigVMDispatch_ArrayAppend(io Array,in Other)"}
   > For_Each_2 = RigVMDispatchNode(Array=@n8.Array) {template="DISPATCH_RigVMDispatch_ArrayIterator(in Array,out Element,out Index,out Count,out Ratio)"}
   > SendEvent = /Script/ControlRig.RigUnit_SendEvent(Event=RequestAutoKey, OffsetInSeconds=0, bEnable=True, bOnlyDuringInteraction=false, Item=@For_Each_2.Element) {is=struct RigVMUnitNode, template="SendEvent::Execute(in Event,in Item,in OffsetInSeconds,in bEnable,in bOnlyDuringInteraction)"}
   &RigUnit_SendEvent(
    Event=RequestAutoKey
    OffsetInSeconds=0
    bEnable=True
    bOnlyDuringInteraction=false
    ExecutePin=@For_Each_2.Completed
    Item=RigVMVariableNode() {variable=Ball Pivot Channel}.Value
   ) {template="SendEvent::Execute(in Event,in Item,in OffsetInSeconds,in bEnable,in bOnlyDuringInteraction)"}
   > &RigUnit_SendEvent(
    Event=RequestAutoKey
    OffsetInSeconds=0
    bEnable=True
    bOnlyDuringInteraction=false
    Item=RigVMVariableNode() {variable=Heel Swing Channel}.Value
   ) {template="SendEvent::Execute(in Event,in Item,in OffsetInSeconds,in bEnable,in bOnlyDuringInteraction)"}
   > &RigUnit_SendEvent(Event=RequestAutoKey, OffsetInSeconds=0, bEnable=True, bOnlyDuringInteraction=false, Item=RigVMVariableNode() {variable=Ball Control}.Value) {template="SendEvent::Execute(in Event,in Item,in OffsetInSeconds,in bEnable,in bOnlyDuringInteraction)"}
   RigVMVariableNode(Value=RigVMVariableNode() {variable=IK Solve}.Value, ExecuteContext=@n2.Completed) {variable=Previous IK Solve}
   n9 = /Script/ControlRig.RigUnit_PreBeginExecution() {is=struct RigVMUnitNode}
   DISPATCH_RigDispatch_GetModuleMetadata = RigVMDispatchNode(Default=true, Name=IK Solve, NameSpace=Parent) {template="DISPATCH_RigDispatch_GetModuleMetadata(in Name,in NameSpace,in Default,out Value,out Found)"}
   RigVMVariableNode(Value=@DISPATCH_RigDispatch_GetModuleMetadata.Value, ExecuteContext=@n9.ExecutePin) {variable=IK Solve}
   > For_Each_1 = RigVMDispatchNode(Array=RigVMVariableNode() {variable=IK Controls}.Value) {template="DISPATCH_RigVMDispatch_ArrayIterator(in Array,out Element,out Index,out Count,out Ratio)"}
   > SetControlVisibility = /Script/ControlRig.RigUnit_SetControlVisibility(Item=@For_Each_1.Element, bVisible=RigVMVariableNode() {variable=IK Solve}.Value) {is=struct RigVMUnitNode, template="SetControlVisibility::Execute(in Item,in Pattern,in bVisible)"}
   > Set Foot Values = RigVMFunctionReferenceNode()
   RigVMFunction_ControlFlowBranch = &RigVMFunction_ControlFlowBranch(ExecuteContext=@"Set Foot Values".ExecuteContext, Condition=RigVMVariableNode() {variable=IK Solve}.Value)
   Set Foot Pivots = RigVMFunctionReferenceNode(ExecuteContext=@RigVMFunction_ControlFlowBranch.True)
   RigUnit_BeginExecution = &RigUnit_BeginExecution()
   RigVMFunction_Sequence_1 = RigVMAggregateNode(ExecuteContext=@RigUnit_BeginExecution.ExecutePin)
   n10 = &RigVMFunction_ControlFlowBranch(ExecuteContext=@RigVMFunction_Sequence_1.B, Condition=RigVMVariableNode() {variable=IK Solve}.Value)
   Make_Array = RigVMDispatchNode(
    Values=((Type=None,Name="None"),(Type=None,Name="None"))
    Values.1=RigVMVariableNode() {variable=Ball Joint}.Value
    Values.0=RigVMVariableNode() {variable=Foot Joint}.Value
   ) {template="DISPATCH_RigVMDispatch_ArrayMake(in Values,out Array)"}
   /Script/ControlRig.RigUnit_FABRIKItemArray(
    MaxIterations=10
    Precision=0.0001
    Weight=1
    bPropagateToChildren=True
    bSetEffectorTransform=false
    EffectorTransform=&RigUnit_GetTransform(bInitial=False, Item=RigVMVariableNode() {variable=Ball IK Null}.Value).Transform
    Items=@Make_Array.Array
    ExecutePin=@n10.True
   ) {is=struct RigVMUnitNode}
   n11 = RigVMVariableNode() {variable=Toe IK Null}
   > &RigUnit_AimConstraintLocalSpaceOffset(
    Filter=(bX=True,bY=True,bZ=True)
    Parents=((Item=(Type=Bone,Name="None"),Weight=1))
    Weight=1
    WorldUp=(Target=(X=0,Y=0,Z=1),Kind=Location,Space=(Type=None,Name="None"))
    Child=RigVMVariableNode() {variable=Ball Joint}.Value
    Parents.0.Item=@n11.Value
    WorldUp.Space=@n11.Value
   )
   n12 = RigVMVariableNode() {variable=Ball Null}
   n13 = RigVMVariableNode() {variable=Ball Joint}
   > &RigUnit_SetTransform(
    Weight=1
    bInitial=False
    Value=/Script/ControlRig.RigUnit_ProjectTransformToNewParent(bChildInitial=True, bNewParentInitial=False, bOldParentInitial=True, Child=@n12.Value, NewParent=@n13.Value, OldParent=@n13.Value) {is=struct RigVMUnitNode}.Transform
    Item=@n12.Value
   )
   n14 = &RigVMFunction_ControlFlowBranch(Condition=RigVMVariableNode() {variable=Has Toes}.Value, ExecuteContext=@RigVMFunction_Sequence_1.E)
   For_Each = RigVMDispatchNode(ExecuteContext=@n14.True, Array=RigVMVariableNode() {variable=Bones}.Value) {template="DISPATCH_RigVMDispatch_ArrayIterator(in Array,out Element,out Index,out Count,out Ratio)"}
   n15 = RigVMDispatchNode(Index=@For_Each.Index, Array=RigVMVariableNode() {variable=FK Controls}.Value) {template="DISPATCH_RigVMDispatch_ArrayGetAtIndex(in Array,in Index,out Element)"}
   > &RigUnit_SetTransform(
    Weight=1
    bInitial=False
    Item=@For_Each.Element
    Value=&RigUnit_ProjectTransformToNewParent(bChildInitial=True, bNewParentInitial=False, bOldParentInitial=True, NewParent=@n15.Element, OldParent=@n15.Element, Child=@For_Each.Element).Transform
   )
   n16 = RigVMVariableNode() {variable=Ball Joint}
   n17 = RigVMVariableNode() {variable=Ball Control}
   &RigUnit_SetTransform(
    Weight=1
    bInitial=False
    Value=&RigUnit_ProjectTransformToNewParent(bChildInitial=True, bNewParentInitial=False, bOldParentInitial=True, Child=@n16.Value, NewParent=@n17.Value, OldParent=@n17.Value).Transform
    Item=@n16.Value
    ExecutePin=@RigVMFunction_Sequence_1.D
   )
   > n18 = RigVMVariableNode(Value=&RigUnit_GetTransform(bInitial=False, Item=RigVMVariableNode() {variable=Ball Control}.Value).Transform) {variable=Current Ball Control Transform}
   n19 = &RigVMFunction_ControlFlowBranch(Condition=/Script/ControlRig.RigUnit_IsInteracting() {is=struct RigVMUnitNode}.bIsInteracting, ExecuteContext=@n18.ExecuteContext)
   n20 = RigVMVariableNode() {variable=Ball Joint}
   n21 = RigVMVariableNode() {variable=Toe Tip Control}
   Get_Metadata = RigVMDispatchNode(Default=(Rotation=(X=0,Y=0,Z=0,W=1),Translation=(X=0,Y=0,Z=0),Scale3D=(X=1,Y=1,Z=1)), Name=Delta, NameSpace=Self, Item=@n21.Value) {template="DISPATCH_RigDispatch_GetMetadata(in Item,in Name,in NameSpace,in Default,out Value,out Found)"}
   Inverse = /Script/RigVM.RigVMFunction_MathTransformInverse(Value=&RigUnit_GetTransform(Space=LocalSpace, bInitial=False, Item=@n21.Value).Transform) {is=struct RigVMUnitNode, template="Inverse::Execute(in Value,out Result)"}
   Multiply = &RigVMFunction_MathTransformMul(B=@Get_Metadata.Value, A=@Inverse.Result) {template="Multiply::Execute(in A,in B,out Result)"}
   RigUnit_ItemToName = /Script/ControlRig.RigUnit_ItemToName(Value=@n21.Value) {is=struct RigVMUnitNode, template="Cast::Execute(in Value,out Result)"}
   SetControlOffset = /Script/ControlRig.RigUnit_SetControlOffset(
    Space=GlobalSpace
    Offset=&RigVMFunction_MathTransformMakeAbsolute(Parent=&RigUnit_GetTransform(bInitial=False, Item=@n20.Value).Transform, Local=@Multiply.Result) {template="Make Absolute::Execute(in Local,in Parent,out Global)"}.Global
    Control=@RigUnit_ItemToName.Result
    ExecutePin=@n19.False
   ) {is=struct RigVMUnitNode, template="SetControlOffset::Execute(in Control,io Offset,in Space)"}
   n22 = RigVMVariableNode() {variable=Foot Joint}
   n23 = RigVMVariableNode() {variable=Heel Control}
   Get_Metadata_1 = RigVMDispatchNode(Default=(Rotation=(X=0,Y=0,Z=0,W=1),Translation=(X=0,Y=0,Z=0),Scale3D=(X=1,Y=1,Z=1)), Name=Delta, NameSpace=Self, Item=@n23.Value) {template="DISPATCH_RigDispatch_GetMetadata(in Item,in Name,in NameSpace,in Default,out Value,out Found)"}
   n24 = &RigVMFunction_MathTransformInverse(Value=&RigUnit_GetTransform(Space=LocalSpace, bInitial=False, Item=@n23.Value).Transform) {template="Inverse::Execute(in Value,out Result)"}
   > &RigUnit_SetControlOffset(
    Space=GlobalSpace
    Offset=&RigVMFunction_MathTransformMakeAbsolute(
     Parent=&RigUnit_GetTransform(bInitial=False, Item=@n22.Value).Transform
     Local=&RigVMFunction_MathTransformMul(B=@Get_Metadata_1.Value, A=@n24.Result) {template="Multiply::Execute(in A,in B,out Result)"}.Result
    ) {template="Make Absolute::Execute(in Local,in Parent,out Global)"}.Global
    Control=&RigUnit_ItemToName(Value=@n23.Value) {template="Cast::Execute(in Value,out Result)"}.Result
   ) {template="SetControlOffset::Execute(in Control,io Offset,in Space)"}
  ]}
]
hierarchy [
 Foot Joint {type=Connector, loc=<0 0 0>, rot=<0 0 0 1>}
 Root {type=Socket, loc=<0 0 0>, rot=<0 0 0 1>}
 Ball Joint {type=Connector, loc=<0 0 0>, rot=<0 0 0 1>}
 Toe Joints {type=Connector, loc=<0 0 0>, rot=<0 0 0 1>}
]
variables [
 BK Buffers {type="TArray<FRigElementKey>"}
 BK Controls {type="TArray<FRigElementKey>"}
 Ball Control {type=FRigElementKey, default=(Type=Bone,Name="")}
 Ball IK Null {type=FRigElementKey, default=(Type=Bone,Name="")}
 Ball Joint {type=FRigElementKey, default=(Type=None,Name="")}
 Ball Null {type=FRigElementKey, default=(Type=Bone,Name="")}
 Ball Pivot {type=float, default=0}
 Ball Pivot Bone {type=FRigElementKey, default=(Type=None,Name="")}
 Ball Pivot Channel {type=FRigElementKey, default=(Type=Bone,Name="")}
 Ball Pivot Null {type=FRigElementKey, default=(Type=None,Name="")}
 Ball Roll {type=float, default=0}
 Bank {type=float, default=0}
 Bones {type="TArray<FRigElementKey>"}
 Channels Control {type=FRigElementKey, default=(Type=Bone,Name="")}
 Control Scale {type=double, default=1}
 Current Ball Control Transform {type=FTransform, default=(Rotation=(X=0,Y=0,Z=0,W=1),Translation=(X=0,Y=0,Z=0),Scale3D=(X=1,Y=1,Z=1))}
 FK Controls {type="TArray<FRigElementKey>"}
 FK Rotation Offset {type=FQuat, default=(X=0.0439,Y=-0.7057,Z=-0.0439,W=-0.7057)}
 Foot Joint {type=FRigElementKey, default=(Type=None,Name="")}
 Has Toes {type=bool, default=False}
 Heel Control {type=FRigElementKey, default=(Type=Bone,Name="")}
 Heel Lift Null {type=FRigElementKey, default=(Type=None,Name="")}
 Heel Pivot {type=float, default=0}
 Heel Pivot Offset {type=FVector, default=(X=0,Y=-21,Z=0)}
 Heel Roll Null {type=FRigElementKey, default=(Type=Bone,Name="")}
 Heel Swing {type=float, default=0}
 Heel Swing Channel {type=FRigElementKey, default=(Type=Bone,Name="")}
 Heel Twist {type=float, default=0}
 IK Controls {type="TArray<FRigElementKey>"}
 IK Null {type=FRigElementKey, default=(Type=Bone,Name="")}
 IK Solve {type=bool, default=False}
 Inner Bank Offset {type=FVector, default=(X=5,Y=0,Z=0)}
 Inner Null {type=FRigElementKey, default=(Type=Bone,Name="")}
 Negative Side {type=bool, default=False}
 Outer Bank Offset {type=FVector, default=(X=-6,Y=0,Z=0)}
 Outer Null {type=FRigElementKey, default=(Type=Bone,Name="")}
 Pivot Shape Settings {
  type=FRigUnit_HierarchyAddControl_ShapeSettings
  default=(bVisible=True,Name="Circle_Thick",Color=(R=1,G=1,B=1,A=1),Transform=(Rotation=(X=-0.7071,Y=0,Z=0,W=0.7071),Translation=(X=0,Y=0,Z=1),Scale3D=(X=1.3,Y=0.8,Z=1)))
 }
 Previous IK Solve {type=bool, default=False}
 Roll {type=double, default=0}
 Roll Blend {type=double, default=0}
 Roll Clamp {type=double, default=0}
 Roll Control {type=FRigElementKey, default=(Type=Bone,Name="")}
 Roll Null {type=FRigElementKey, default=(Type=Bone,Name="")}
 Running Parent {type=FRigElementKey, default=(Type=Bone,Name="")}
 Shape Settings {
  type=FRigUnit_HierarchyAddControl_ShapeSettings
  default=(bVisible=True,Name="Circle_Thick",Color=(R=1,G=1,B=1,A=1),Transform=(Rotation=(X=-0.7071,Y=0,Z=0,W=0.7071),Translation=(X=0,Y=0,Z=1),Scale3D=(X=1.3,Y=0.8,Z=1)))
 }
 Target Effector {type=FRigElementKey, default=(Type=Bone,Name="")}
 Target IK Control {type=FRigElementKey, default=(Type=Bone,Name="")}
 Toe IK End Null {type=FRigElementKey, default=(Type=Bone,Name="")}
 Toe IK Null {type=FRigElementKey, default=(Type=Bone,Name="")}
 Toe Tip Control {type=FRigElementKey, default=(Type=Bone,Name="")}
 Toe Tip Pivot Null {type=FRigElementKey, default=(Type=None,Name="")}
 Toe Tip Pivot Offset {type=FVector, default=(X=0,Y=8,Z=0)}
 Toe Tip Roll {type=float, default=0}
 Toe Tip Swing {type=float, default=0}
]
